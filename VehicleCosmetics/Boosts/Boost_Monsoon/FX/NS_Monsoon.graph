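# Niagara system for the Monsoon vehicle boost cosmetic: a storm exhaust of clouds, mist, electricity sprites and a rain ribbon trail streaming backward.
# Five emitters distribute particles across the vehicle's skeletal mesh sockets by port count, with per-emitter painted color variants and a product preview module.
# Electricity and Local_Clouds each merge two looks through spawn groups, switching parameters and setting visibility tags by group.
# InheritSourceMovement is present but disabled throughout.
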NiagaraSystem /VehicleCosmetics/Boosts/Boost_Monsoon/FX/NS_Monsoon
dependencies {
 usedDataInterfaces [
  &NiagaraDataInterfaceCurve
  &NiagaraDataInterfaceArrayInt32
  &NiagaraDataInterfaceVectorCurve
  &NiagaraDataInterfaceSpriteRendererInfo
  &NiagaraDataInterfaceSkeletalMesh
  &NiagaraDataInterfaceArrayFloat3
 ]
 usedDynamicInputs [&SelectVectorFromArray]
 usedModules [
  &self:SystemSpawnScript.SystemScriptSource.SystemScriptGraph.NiagaraNodeAssignment_0.SetVariables_#1
  &self:SystemSpawnScript.SystemScriptSource.SystemScriptGraph.NiagaraNodeAssignment_1.SetVariables_#2
  &self:SystemSpawnScript.SystemScriptSource.SystemScriptGraph.NiagaraNodeAssignment_2.SetVariables_#3
  &SystemState
 ]
}
emitters {
 Electricity {
  emitterUpdateScript [
   &EmitterState()
   &VC_FilteredSocketCount()
   &SpawnRate(SpawnRate=50, Use Spawn Probability=false)
   &SpawnRate(SpawnGroup=1, SpawnRate=20, Use Spawn Probability=false) {module=SpawnRate002}
  ]
  particleSpawnScript [
   &VC_PaintedVariantsOverride(
    Black=<rgba 750 750 750 1>
    BurntSienna=<rgba 750 231.5642 92.4481 1>
    Cobalt=<rgba 0 13.8751 750 1>
    Crimson=<rgba 750 0 20.3283 1>
    DefaultColor=<rgba 600 420.3152 0 1>
    ForestGreen=<rgba 0 750 37.8946 1>
    Grey=<rgba 750 750 750 1>
    Lime=<rgba 375 750 0 1>
    Orange=<rgba 750 261.6032 0 1>
    Pink=<rgba 750 89.8442 509.9429 1>
    Purple=<rgba 307.8107 0 750 1>
    Saffron=<rgba 750 575 0 1>
    SkyBlue=<rgba 0 499.04 750 1>
    TitaniumWhite=<rgba 750 750 750 1>
   )
   &InitializeParticle(
    Color Mode=::NewEnumerator0
    Lifetime=&MergeHelper_FloatSwitch(
     0=&RandomRangeFloat(Minimum=0.1, Maximum=0.15)
     1=0.45
     2=?
     3=?
     Index=&SpawnGroup()
     Num=::NewEnumerator0
    )
    Sprite Rotation Mode=::NewEnumerator2
    Sprite Size Mode=::NewEnumerator3
    Sprite UV Mode=::NewEnumerator3
    Uniform Sprite Size=&MergeHelper_FloatSwitch(0=75, 1=50, 2=?, 3=?, Index=&SpawnGroup(), Num=::NewEnumerator0)
    UsePositionOffset=false
    Write Material Random=true
   )
   &PartitionParticles(Number Of Partitions=$Emitter.PortCount, Partition Mode:ENiagara_PartitionMode=::NewEnumerator1)
   &SkeletalMeshLocation(
    Bone / Socket Index=$Particles.PartitionParticles.Partition
    Bone / Socket Sampling Mode=::NewEnumerator5
    MaskSpawnGroup=false
    Mesh Sampling Type=::NewEnumerator6
    Skeletal Mesh=$System.SkeletalMesh
   )
   &AddVelocity(Velocity=&MergeHelper_VectorSwitch(0=<-500 0 0>, 1=<-1500 0 0>, 2=?, 3=?, Index=&SpawnGroup(), Num=::NewEnumerator0))
   &ShapeLocation(
    MaskSpawnGroup=false
    Override Seed=false
    Sphere Radius=&MergeHelper_FloatSwitch(0=60, 1=40, 2=?, 3=?, Index=&SpawnGroup(), Num=::NewEnumerator0)
    Sphere Surface Distribution=1
    UseOwnerScale=true
   )
   /SetVariables(Particles.VisibilityTag=&SelectIntFromArray(Direct Array Index=&SpawnGroup(), Int Selection Array=DataInterface() {propertyValues {IntData [0,1,2,3,3]}})) {isSetParameters=true}
   &InheritVelocity(Inherited Velocity Speed Limit=250, Speed Limit=true, UseSpeedThreshold=false)
   /VehicleCosmetics/FX/VC_FXParents/Modules/DLR_ProductPreviews(GameModeID:NiagaraInt32=$System.GameModeID, bPreviewVehicleSpeed:NiagaraBool=$System.bPreviewVehicleSpeed)
  ]
  particleUpdateScript [
   &ParticleState()
   &InheritSourceMovement(
    Applied Position Scale=&Multiply_VectorByFloat(
     Vector=<1 1 1>
     Float=&FloatFromCurve(FloatCurve=DataInterface() {propertyValues {Curve {
         defaultValue=1
         keys [
          {arriveTangent=-1.7885, interpMode=RCIM_Cubic, leaveTangent=-1.7885, tangentMode=RCTM_User, time=0.05, value=1}
          {interpMode=RCIM_Cubic, tangentMode=RCTM_User, time=1, value=0}
         ]
        }}})
    )
    ApplyForce=false
    ApplyPosition=true
    ApplyPositionFalloff=false
   ) {enabled=false, module=InheritSourceMovement001}
   &SubUVAnimation(
    "Sprite Renderer "=DataInterface() {propertyValues {SpriteRenderer=&self:Electricity_0.NiagaraSpriteRendererProperties_2}}
    UseEndFrame=false
    UseStartFrame=false
   )
   &SolveForcesAndVelocity(Clamp Velocity=false, Limit Acceleration=false)
  ]
  rendererClasses [&NiagaraSpriteRendererProperties]
  renderers [
   {rendererClass=&NiagaraSpriteRendererProperties}
   {rendererClass=&NiagaraSpriteRendererProperties, rendererIndex=1}
  ]
 }
 Local_Clouds {
  emitterUpdateScript [
   &EmitterState()
   &VC_FilteredSocketCount()
   &SpawnRate(SpawnRate=50, Use Spawn Probability=false)
   &SpawnRate(SpawnGroup=1, SpawnRate=50, Use Spawn Probability=false) {module=SpawnRate002}
  ]
  particleSpawnScript [
   &VC_PaintedVariantsOverride(
    Black=<rgba 0.005 0.005 0.005 1>
    BurntSienna=<rgba 0.3185 0.1195 0.0666 1>
    DefaultColor=&MergeHelper_ColorSwitch(
     0=<rgba 0.0723 0.1441 0.4508 1>
     1=<rgba 0.7812 0.9256 1 0.5>
     2=?
     3=?
     Index=&SpawnGroup()
     Num=::NewEnumerator0
    )
    Gold:LinearColor=<rgba 1 0.4401 0 1>
    Grey=<rgba 0.1119 0.1119 0.1356 1>
    Pink=<rgba 0.3916 0.1683 0.4969 1>
    Purple=<rgba 0.2907 0 0.7083 1>
    TitaniumWhite=<rgba 1 1 1 1>
   )
   &InitializeParticle(
    Color Mode=::NewEnumerator0
    Lifetime=&MergeHelper_FloatSwitch(
     0=&RandomRangeFloat(Minimum=0.15, Maximum=0.2)
     1=&RandomRangeFloat(Minimum=0.5, Maximum=0.75)
     2=?
     3=?
     Index=&SpawnGroup()
     Num=::NewEnumerator0
    )
    Mesh Scale Max=<1.5 1.25 1.5>
    Mesh Scale Min=<1 0.75 1>
    Mesh Scale Mode=::NewEnumerator2
    Sprite Size=&MergeHelper_Vector2Switch(0=<0 0>, 1=<10 15>, 2=?, 3=?, 4=?, Index=&SpawnGroup(), Num=::NewEnumerator0)
    Sprite Size Mode=::NewEnumerator4
    Sprite UV Mode=::NewEnumerator1
    UsePositionOffset=false
    Write Material Random=true
   )
   &PartitionParticles(Number Of Partitions=$Emitter.PortCount, Partition Mode:ENiagara_PartitionMode=::NewEnumerator5)
   &SkeletalMeshLocation(
    Bone / Socket Index=$Particles.PartitionParticles.Partition
    Bone / Socket Sampling Mode=::NewEnumerator5
    MaskSpawnGroup=false
    Mesh Sampling Type=::NewEnumerator6
    Skeletal Mesh=$System.SkeletalMesh
   )
   &AddVelocity(Velocity=&MergeHelper_VectorSwitch(0=<-1000 0 0>, 1=<0 0 0>, 2=?, 3=?, Index=&SpawnGroup(), Num=::NewEnumerator0))
   &AddVelocity(
    Cone Angle=180
    Cone Axis=<-1 0 0>
    Inner Cone Angle=120
    Module.UseConeFalloff=false
    Module.UseVelDistribution=false
    Override Seed=false
    Velocity Mode=::NewEnumerator2
    Velocity Speed=&MergeHelper_FloatSwitch(0=0, 1=550, 2=?, 3=?, Index=&SpawnGroup(), Num=::NewEnumerator0)
   ) {module=AddVelocity001}
   &ShapeLocation(
    MaskSpawnGroup=false
    Offset=&MergeHelper_VectorSwitch(0=<-40 0 0>, 1=<30 0 0>, 2=?, 3=?, Index=&SpawnGroup(), Num=::NewEnumerator0)
    Offset Coordinate Space=::Local
    Offset Mode=::NewEnumerator0
    Override Seed=false
    Sphere Radius=&MergeHelper_FloatSwitch(0=40, 1=20, 2=?, 3=?, Index=&SpawnGroup(), Num=::NewEnumerator0)
    UseOwnerScale=true
   )
   &InitialMeshOrientation(
    Mesh Orientation Mode=::NewEnumerator1
    Rotation=&RandomRangeVector(Minimum=<0 0 0>, Maximum=<1 1 1>)
    Use Rotation Vector=true
   )
   /SetVariables(Particles.VisibilityTag=&SelectIntFromArray(Direct Array Index=&SpawnGroup(), Int Selection Array=DataInterface() {propertyValues {IntData [0,1,2,3,3]}})) {isSetParameters=true}
   &DLR_ProductPreviews(GameModeID:NiagaraInt32=$System.GameModeID, bPreviewVehicleSpeed:NiagaraBool=$System.bPreviewVehicleSpeed)
  ]
  particleUpdateScript [
   &ParticleState()
   &ScaleMeshSize(Scale Factor=&VectorFromCurve(VectorCurve=DataInterface() {propertyValues {
       XCurve {keys [
         {arriveTangent=6.5452, interpMode=RCIM_Cubic, leaveTangent=6.5452, tangentMode=RCTM_User, time=0, value=0}
         {arriveTangent=1, interpMode=RCIM_Cubic, leaveTangent=1, time=0.2, value=0.75}
         <at 1 1 RCIM_Cubic>
        ]}
       YCurve {keys [
         {arriveTangent=6.5452, interpMode=RCIM_Cubic, leaveTangent=6.5452, tangentMode=RCTM_User, time=0, value=0}
         {arriveTangent=1, interpMode=RCIM_Cubic, leaveTangent=1, time=0.2, value=0.75}
         <at 1 1 RCIM_Cubic>
        ]}
       ZCurve {keys [
         {arriveTangent=6.5452, interpMode=RCIM_Cubic, leaveTangent=6.5452, tangentMode=RCTM_User, time=0, value=0}
         {arriveTangent=1, interpMode=RCIM_Cubic, leaveTangent=1, time=0.2, value=0.75}
         <at 1 1 RCIM_Cubic>
        ]}
      }}))
   &AccelerationForce(
    Acceleration=&VectorFromCurve(
     VectorCurve=DataInterface() {propertyValues {
       XCurve {keys [{arriveTangent=2.5196, interpMode=RCIM_Cubic, leaveTangent=2.5196, tangentMode=RCTM_User, time=0, value=0}, <at 1 1 RCIM_Cubic>]}
       YCurve {keys [<at 0 0>,<at 1 0>]}
       ZCurve {keys [<at 0 0>,<at 1 0>]}
      }}
     Scale Curve=&MergeHelper_VectorSwitch(0=<0 0 0>, 1=<-5000 0 0>, 2=?, 3=?, Index=&SpawnGroup(), Num=::NewEnumerator0)
    )
    Coordinate Space=::Local
   )
   &Drag(
    Drag=&FloatFromCurve(
     FloatCurve=DataInterface() {propertyValues {Curve {keys [<at 0 1>,<at 1 0>]}}}
     Scale Curve=&MergeHelper_FloatSwitch(0=0, 1=5, 2=?, 3=?, Index=&SpawnGroup(), Num=::NewEnumerator0)
    )
    Use Linear Drag=true
    Use Rotational Drag=true
   )
   &AddRotationalVelocity(Rotation Rate=<0 -0.05 0>)
   &InheritSourceMovement(
    Applied Position Scale=&Multiply_VectorByFloat(
     Vector=<1 1 1>
     Float=&FloatFromCurve(FloatCurve=DataInterface() {propertyValues {Curve {
         defaultValue=1
         keys [
          {arriveTangent=-1.7885, interpMode=RCIM_Cubic, leaveTangent=-1.7885, tangentMode=RCTM_User, time=0.05, value=1}
          {interpMode=RCIM_Cubic, tangentMode=RCTM_User, time=1, value=0}
         ]
        }}})
    )
    ApplyForce=false
    ApplyPosition=true
    ApplyPositionFalloff=false
   ) {enabled=false, module=InheritSourceMovement001}
   &DynamicMaterialParameters(
    Index 0 Param 1=&MergeHelper_FloatSwitch(0=0, 1=1, 2=?, 3=?, Index=&SpawnGroup(), Num=::NewEnumerator0)
    Index 0 Param 2=&MergeHelper_FloatSwitch(
     0=0
     1=&FloatFromCurve(FloatCurve=DataInterface() {propertyValues {Curve {keys [<at 0 1>,<at 1 0>]}}})
     2=?
     3=?
     Index=&SpawnGroup()
     Num=::NewEnumerator0
    )
    Index 0 Param 3=&MergeHelper_FloatSwitch(0=0, 1=1, 2=?, 3=?, Index=&SpawnGroup(), Num=::NewEnumerator0)
    Index 0 Param 4=&MergeHelper_FloatSwitch(
     0=&FloatFromCurve(FloatCurve=DataInterface() {propertyValues {Curve {
         defaultValue=1
         keys [<at 0 0 RCIM_Cubic>, {arriveTangent=3.385, interpMode=RCIM_Cubic, leaveTangent=3.385, tangentMode=RCTM_User, time=1, value=1}]
        }}})
     1=0
     2=?
     3=?
     Index=&SpawnGroup()
     Num=::NewEnumerator0
    )
    Index 1 Param 2=1
    Module.Index0Float0Write=true
    Module.Index0Float1Write=true
    Module.Index0Float2Write=true
    Module.Index0Float3Write=true
    Module.Index1Float0Write=false
    Module.Index1Float1Write=true
    Module.Index1Float2Write=true
    Module.Index1Float3Write=true
    Write Parameter Index 1=::NewEnumerator1
   )
   &ScaleColor(
    Scale Alpha=&FloatFromCurve(FloatCurve=DataInterface() {propertyValues {Curve {
        defaultValue=1
        keys [
         {tangentMode=RCTM_User, time=0, value=1}
         {arriveTangent=-0.0405, interpMode=RCIM_Cubic, leaveTangent=-0.0405, tangentMode=RCTM_User, time=0.7009, value=1}
         {arriveTangent=-8.0089, interpMode=RCIM_Cubic, leaveTangent=-8.0089, tangentMode=RCTM_User, time=1, value=0}
        ]
       }}})
    ScaleA=true
    ScaleRGB=true
   )
   &SolveForcesAndVelocity(Clamp Velocity=false, Limit Acceleration=false)
   &ScaleSpriteSizeBySpeed(Max Scale Factor=<1 2>)
  ]
  rendererClasses [&NiagaraMeshRendererProperties, &NiagaraSpriteRendererProperties]
  renderers [
   {rendererClass=&NiagaraMeshRendererProperties, rendererIndex=0}
   {rendererClass=&NiagaraSpriteRendererProperties, rendererIndex=1}
  ]
 }
 Mist {
  emitterUpdateScript [
   &EmitterState()
   &VC_FilteredSocketCount()
   &SpawnRate(SpawnRate=30, Use Spawn Probability=false)
  ]
  particleSpawnScript [
   &VC_PaintedVariantsOverride(
    Black=<rgba 0.0191 0.0191 0.0191 0.25>
    BurntSienna=<rgba 0.4896 0.1512 0.0603 0.25>
    Cobalt=<rgba 0 0.0185 1 0.25>
    Crimson=<rgba 0.4635 0 0.0126 0.25>
    DefaultColor=<rgba 1 1 1 0.25>
    ForestGreen=<rgba 0 0.5 0.0253 0.25>
    Grey=<rgba 0.3802 0.3802 0.3802 0.25>
    Lime=<rgba 0.2943 0.5885 0 0.25>
    Orange=<rgba 0.6724 0.2346 0 0.25>
    Pink=<rgba 0.8229 0.0986 0.5595 0.25>
    Purple=<rgba 0.2789 0 0.6795 0.25>
    Saffron=<rgba 0.9 0.69 0 0.25>
    SkyBlue=<rgba 0 0.357 0.5365 0.25>
    TitaniumWhite=<rgba 0.75 0.75 0.75 0.25>
   )
   &InitializeParticle(
    Color Mode=::NewEnumerator0
    Lifetime=0.65
    Sprite Rotation Mode=::NewEnumerator2
    Sprite Size Mode=::NewEnumerator1
    Sprite UV Mode=::NewEnumerator3
    Uniform Sprite Size Max=200
    Uniform Sprite Size Min=175
    UsePositionOffset=false
    Write Material Random=true
   )
   &PartitionParticles(Number Of Partitions=$Emitter.PortCount)
   &SkeletalMeshLocation(
    Bone / Socket Index=$Particles.PartitionParticles.Partition
    Bone / Socket Sampling Mode=::NewEnumerator5
    MaskSpawnGroup=false
    Mesh Sampling Type=::NewEnumerator6
    Skeletal Mesh=$System.SkeletalMesh
   )
   &AddVelocity(Velocity=<-1500 0 50>)
   &InheritVelocity(Inherited Velocity Speed Limit=250, Speed Limit=true, UseSpeedThreshold=false) {enabled=false}
   &DLR_ProductPreviews(GameModeID:NiagaraInt32=$System.GameModeID, bPreviewVehicleSpeed:NiagaraBool=$System.bPreviewVehicleSpeed)
   &ShapeLocation(MaskSpawnGroup=false, Override Seed=false, Sphere Radius=20, UseOwnerScale=true)
  ]
  particleUpdateScript [
   &ParticleState()
   &ScaleSpriteSize(Uniform Curve Sprite Scale=DataInterface() {propertyValues {Curve {
       defaultValue=0.2
       keys [
        {arriveTangent=4.1785, interpMode=RCIM_Cubic, leaveTangent=4.1785, tangentMode=RCTM_User, time=0, value=0}
        {arriveTangent=1.1179, interpMode=RCIM_Cubic, leaveTangent=1.1179, tangentMode=RCTM_User, time=0.2042, value=0.6228}
        {arriveTangent=0.0224, interpMode=RCIM_Cubic, leaveTangent=0.0224, tangentMode=RCTM_User, time=1, value=1}
       ]
      }}})
   &InheritSourceMovement(
    Applied Position Scale=&Multiply_VectorByFloat(
     Vector=<1 1 1>
     Float=&FloatFromCurve(FloatCurve=DataInterface() {propertyValues {Curve {
         defaultValue=1
         keys [
          {arriveTangent=-1.7885, interpMode=RCIM_Cubic, leaveTangent=-1.7885, tangentMode=RCTM_User, time=0.05, value=1}
          {interpMode=RCIM_Cubic, tangentMode=RCTM_User, time=1, value=0}
         ]
        }}})
    )
    ApplyForce=false
    ApplyPosition=true
    ApplyPositionFalloff=false
   ) {enabled=false, module=InheritSourceMovement001}
   &SubUVAnimation(
    "Sprite Renderer "=DataInterface() {propertyValues {SpriteRenderer=&self:Mist_0.NiagaraSpriteRendererProperties_2}}
    SubUV Animation Mode:ENiagara_SubUVLookupModeV2=::NewEnumerator2
    UseEndFrame=false
    UseStartFrame=false
   )
   &ScaleColor(
    Scale Alpha=&FloatFromCurve(FloatCurve=DataInterface() {propertyValues {Curve {
        defaultValue=1
        keys [
         {arriveTangent=-1.3384, interpMode=RCIM_Cubic, leaveTangent=-1.3384, tangentMode=RCTM_User, time=0, value=0.75}
         {interpMode=RCIM_Cubic, tangentMode=RCTM_User, time=1, value=0}
        ]
       }}})
    ScaleA=true
    ScaleRGB=true
   )
   &SolveForcesAndVelocity(Clamp Velocity=false, Limit Acceleration=false)
  ]
  rendererClasses [&NiagaraSpriteRendererProperties]
  renderers [{rendererClass=&NiagaraSpriteRendererProperties}]
 }
 Rain_Trail {
  emitterUpdateScript [
   &EmitterState()
   &VC_FilteredSocketCount()
   &VC_RateQualityHandler()
   &VC_SpawnRibbon_v1(Number of Ribbons=$Emitter.PortCount, SpawnRate=$Emitter.Spawn Rate)
  ]
  particleSpawnScript [
   &VC_PaintedVariantsOverride(
    Black=<rgba 0 0 0 50>
    BurntSienna=<rgba 50 15.4376 6.1632 1>
    Cobalt=<rgba 0 0.925 50 1>
    Crimson=<rgba 50 0 1.3552 1>
    DefaultColor=<rgba 4.2708 12.4082 20 1>
    ForestGreen=<rgba 0 50 2.5263 1>
    Gold:LinearColor=<rgba 50 50 50 1>
    Grey=<rgba 50 50 50 1>
    Lime=<rgba 25 50 0 1>
    Orange=<rgba 50 17.4402 0 1>
    Pink=<rgba 50 5.9896 33.9962 1>
    Purple=<rgba 20.5207 0 50 1>
    Saffron=<rgba 50 38.3333 0 1>
    SkyBlue=<rgba 0 33.2693 50 1>
    TitaniumWhite=<rgba 50 50 50 1>
   )
   &VC_InitializeRibbon_v1(Color=$Particles.Color, Enable Skeletal Position=true, Lifetime=0.45, Ribbon Width=100)
   &AddVelocity(Velocity=<-4000 0 0>) {module=AddVelocity001}
   &DLR_ProductPreviews(GameModeID:NiagaraInt32=$System.GameModeID, bPreviewVehicleSpeed:NiagaraBool=$System.bPreviewVehicleSpeed)
  ]
  particleUpdateScript [
   &ParticleState()
   &InheritVelocity(Inherited Velocity Speed Limit=50, Speed Limit=true, UseSpeedThreshold=false)
   &InheritSourceMovement(
    Applied Position Scale=&VectorFromFloat(Value=&FloatFromCurve(FloatCurve=DataInterface() {propertyValues {Curve {
         defaultValue=1
         keys [
          {arriveTangent=-1.7885, interpMode=RCIM_Cubic, leaveTangent=-1.7885, tangentMode=RCTM_User, time=0.05, value=1}
          {interpMode=RCIM_Cubic, tangentMode=RCTM_User, time=1, value=0}
         ]
        }}}))
    ApplyForce=false
    ApplyPosition=true
    ApplyPositionFalloff=false
   ) {enabled=false}
   &SolveForcesAndVelocity(Clamp Velocity=false, Limit Acceleration=false)
   &ScaleColor(
    Scale Alpha=&FloatFromCurve(FloatCurve=DataInterface() {propertyValues {Curve {keys [<at 0 1 RCIM_Cubic>,<at 1 0 RCIM_Cubic>]}}})
    ScaleA=true
    ScaleRGB=true
   )
   &ScaleRibbonWidth(Ribbon Width Scale=&FloatFromCurve(FloatCurve=DataInterface() {propertyValues {Curve {keys [<at 0 1 RCIM_Cubic>,<at 1 0.75 RCIM_Cubic>]}}}))
   &VC_UpdateRibbon_v1(Enable Skeletal Position=true)
  ]
  rendererClasses [&NiagaraRibbonRendererProperties]
  renderers [{rendererClass=&NiagaraRibbonRendererProperties}]
 }
 Trailing_Clouds {
  emitterUpdateScript [
   &EmitterState()
   &VC_FilteredSocketCount()
   &SpawnRate(SpawnRate=75, Use Spawn Probability=false)
  ]
  particleSpawnScript [
   &VC_PaintedVariantsOverride(
    Black=<rgba 1 1 1 1>
    BurntSienna=<rgba 1 1 1 1>
    Cobalt=<rgba 1 1 1 1>
    Crimson=<rgba 1 1 1 1>
    ForestGreen=<rgba 1 1 1 1>
    Grey=<rgba 1 1 1 1>
    Lime=<rgba 1 1 1 1>
    Orange=<rgba 1 1 1 1>
    Pink=<rgba 1 1 1 1>
    Purple=<rgba 1 1 1 1>
    R1 Black:LinearColor=<rgba 0 0 0 1>
    R1 BurntSienna:LinearColor=<rgba 0.1094 0.041 0.0229 1>
    R1 Cobalt:LinearColor=<rgba 0 0.0097 0.526 1>
    R1 Crimson:LinearColor=<rgba 0.2396 0 0.0065 1>
    R1 DefaultColor=<rgba 0.0334 0.0666 0.2083 1>
    R1 ForestGreen:LinearColor=<rgba 0 0.224 0.0113 1>
    R1 Gold:LinearColor=<rgba 0.4479 0.2135 0 1>
    R1 Grey:LinearColor=<rgba 0.0344 0.0344 0.0417 1>
    R1 Lime:LinearColor=<rgba 0.1224 0.2448 0 1>
    R1 Orange:LinearColor=<rgba 0.349 0.1217 0 1>
    R1 Pink:LinearColor=<rgba 0.1949 0.0829 0.2448 1>
    R1 Purple:LinearColor=<rgba 0.1496 0 0.3646 1>
    R1 Saffron:LinearColor=<rgba 0.4115 0.3155 0 1>
    R1 SkyBlue:LinearColor=<rgba 0 0.1733 0.2604 1>
    R1 TitaniumWhite:LinearColor=<rgba 1 1 1 1>
    R2 BurntSienna:LinearColor=<rgba 0.3185 0.1195 0.0666 1>
    R2 DefaultColor:LinearColor=<rgba 0.0723 0.1441 0.4508 1>
    R2 Gold:LinearColor=<rgba 1 0.3994 0 1>
    R2 Grey:LinearColor=<rgba 0.1119 0.1119 0.1356 1>
    R2 Pink:LinearColor=<rgba 0.3916 0.1683 0.4969 1>
    R2 Purple:LinearColor=<rgba 0.3313 0 0.8073 1>
    R2 TitaniumWhite:LinearColor=<rgba 2 2 2 1>
    Saffron=<rgba 1 1 1 1>
    SkyBlue=<rgba 1 1 1 1>
    Use R1 Painted Overrides=true
    Use R2 Painted Overrides=true
   )
   &InitializeParticle(
    Color Mode=::NewEnumerator0
    Lifetime Max=0.6
    Lifetime Min=0.4
    Lifetime Mode=::NewEnumerator1
    Sprite Size Mode=::NewEnumerator3
    Sprite UV Mode=::NewEnumerator1
    Uniform Sprite Size=300
    UsePositionOffset=false
    Write Material Random=true
   )
   &PartitionParticles(Number Of Partitions=$Emitter.PortCount)
   &SkeletalMeshLocation(
    Bone / Socket Index=$Particles.PartitionParticles.Partition
    Bone / Socket Sampling Mode=::NewEnumerator5
    MaskSpawnGroup=false
    Mesh Sampling Type=::NewEnumerator6
    Skeletal Mesh=$System.SkeletalMesh
   )
   &AddVelocity(Velocity=<-2500 0 0>)
   &ShapeLocation(
    MaskSpawnGroup=false
    Offset=<-40 0 0>
    Offset Coordinate Space=::Local
    Offset Mode=::NewEnumerator0
    Override Seed=false
    Sphere Radius=30
    UseOwnerScale=true
   )
   &InheritVelocity(Inherited Velocity Speed Limit=500, Speed Limit=true, UseSpeedThreshold=false) {enabled=false}
   &DLR_ProductPreviews(GameModeID:NiagaraInt32=$System.GameModeID, bPreviewVehicleSpeed:NiagaraBool=$System.bPreviewVehicleSpeed)
  ]
  particleUpdateScript [
   &ParticleState()
   &ScaleSpriteSize(Uniform Curve Sprite Scale=DataInterface() {propertyValues {Curve {
       defaultValue=1
       keys [
        {interpMode=RCIM_Cubic, tangentMode=RCTM_User, time=0, value=0.75}
        {arriveTangent=-1.2756, interpMode=RCIM_Cubic, leaveTangent=-1.2756, tangentMode=RCTM_User, time=1, value=0}
       ]
      }}})
   &Drag(Use Linear Drag=true, Use Rotational Drag=true)
   &InheritSourceMovement(
    Applied Position Scale=&Multiply_VectorByFloat(
     Vector=<1 1 1>
     Float=&FloatFromCurve(FloatCurve=DataInterface() {propertyValues {Curve {
         defaultValue=1
         keys [
          {arriveTangent=-1.7885, interpMode=RCIM_Cubic, leaveTangent=-1.7885, tangentMode=RCTM_User, time=0.05, value=1}
          {interpMode=RCIM_Cubic, tangentMode=RCTM_User, time=1, value=0}
         ]
        }}})
    )
    ApplyForce=false
    ApplyPosition=true
    ApplyPositionFalloff=false
   ) {enabled=false, module=InheritSourceMovement001}
   &SubUVAnimation(
    "Sprite Renderer "=DataInterface() {propertyValues {SpriteRenderer=&self:Trailing_Clouds_0.NiagaraSpriteRendererProperties_2}}
    UseEndFrame=false
    UseStartFrame=false
   )
   &DynamicMaterialParameters(
    Index 0 Param 1=1
    Index 0 Param 2=&FloatFromCurve(FloatCurve=DataInterface() {propertyValues {Curve {keys [<at 0 1>,<at 1 0>]}}})
    Index 0 Param 3=1
    Index 1 Param 2=1
    Index 2 Linear Color:LinearColor=$Particles.R1 Color
    Index 3 Linear Color:LinearColor=$Particles.R2 Color
    Module.Index0Float0Write=true
    Module.Index0Float1Write=true
    Module.Index0Float2Write=true
    Module.Index0Float3Write=true
    Module.Index1Float0Write=false
    Module.Index1Float1Write=true
    Module.Index1Float2Write=true
    Module.Index1Float3Write=true
    Write Parameter Index 1=::NewEnumerator1
    Write Parameter Index 2=::NewEnumerator5
    Write Parameter Index 3=::NewEnumerator5
   )
   &SolveForcesAndVelocity(Clamp Velocity=false, Limit Acceleration=false)
  ]
  rendererClasses [&NiagaraSpriteRendererProperties]
  renderers [{rendererClass=&NiagaraSpriteRendererProperties}]
 }
}
summary {
 systemName=NS_Monsoon
 emitters [
  {emitterName=Rain_Trail, rendererClasses [&NiagaraRibbonRendererProperties]}
  {emitterName=Local_Clouds, rendererClasses [&NiagaraMeshRendererProperties, &NiagaraSpriteRendererProperties]}
  {emitterName=Electricity, rendererClasses [&NiagaraSpriteRendererProperties]}
  {emitterName=Mist, rendererClasses [&NiagaraSpriteRendererProperties]}
  {emitterName=Trailing_Clouds, rendererClasses [&NiagaraSpriteRendererProperties]}
 ]
 userVariables [
  User.bIsFrontendPreview {type=NiagaraBool}
  User.bIsPainted {type=NiagaraBool}
  User.blsIsCoroColor {type=NiagaraBool}
  User.bPreviewVehicleSpeed {type=NiagaraBool}
  User.ColorVariantID {type=NiagaraInt32}
  User.GameModeID {type=NiagaraInt32}
  User.PaintedColor {type=LinearColor, description="", defaultValue {struct=LinearColor, value=<rgba 1 1 1 1>}}
  User.PaintedScalar {type=NiagaraFloat, defaultValue=1, description=""}
  User.PortCount {type=NiagaraInt32, defaultValue=2}
 ]
}
